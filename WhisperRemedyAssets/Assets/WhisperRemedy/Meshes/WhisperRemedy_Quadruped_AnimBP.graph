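# Animation Blueprint for the WhisperRemedy quadruped skeleton that makes the mesh follow the pose of the mesh it is attached to.
# A Copy Pose From Mesh node, set to use the attached parent, feeds a DefaultSlot montage slot that drives the output pose.
# The event graph is empty and there are no variables, so it holds no locomotion or state logic of its own.
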AnimBlueprint /WhisperRemedyAssets/Assets/WhisperRemedy/Meshes/WhisperRemedy_Quadruped_AnimBP
target_skeleton=&/WhisperRemedy_Quadruped_Skeleton
graphs [
 AnimGraph {nodes [Output Pose(Result=Slot 'DefaultSlot'(Source=Copy Pose From Mesh() {bUseAttachedParent=true}))]}
 EventGraph {links [], nodes []}
]
variables []
