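# Post-process AnimBP for the WhisperRemedy quadruped, adding face poses and collar tag motion to the input pose.
# The pose asset and DefaultSlot feed a weighted neck, head and spine constraint on the tag master bone, then sprung AnimDynamics swings the tag.
# Each update looks up the ShirtTilapia Chaos cloth interactor and, with cloth enabled, overrides its gravity using a hips-aligned vector.
# It also sets RbanOrCloth from the rigid-body and cloth flags.
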AnimBlueprint /WhisperRemedyAssets/Assets/WhisperRemedy/Meshes/WhisperRemedy_Quadruped_AnimBP_Post
target_skeleton=&/WhisperRemedy_Quadruped_Skeleton
graphs [
 AnimGraph {nodes [
   n1 = WhisperRemedy_Quadruped_FacePoses_PoseAsset!(SourcePose:Pose Link Structure=Input Pose()) -> (Pose:Pose Link Structure) {
    is=title AnimGraphNode_PoseBlendNode
    blendOption=Linear
    poseAsset=&/WhisperRemedy_Quadruped_FacePoses_PoseAsset
   }
   n2 = "Constraint - dyn_collar_tag_MSTR (PPPPP)"!(
    ComponentPose=Local To Component(LocalPose=Slot 'DefaultSlot'(Source=@n1))
    Alpha=1
    ConstraintWeights_0=0.4911
    ConstraintWeights_1:float=0.3552
    ConstraintWeights_2:float=0.1118
    ConstraintWeights_3:float=0.0296
    ConstraintWeights_4:float=0.0123
   ) -> (Pose) {
    alphaBoolBlend {blendOption=Linear}
    boneToModify {boneName=dyn_collar_tag_MSTR}
    constraintSetup [
     {targetBone {boneName=neck_02}}
     {targetBone {boneName=neck_01}}
     {targetBone {boneName=head}}
     {targetBone {boneName=spine_03}}
     {targetBone {boneName=spine_02}}
    ]
    constraintWeights [1,1,1,1,1]
   }
   n3 = Anim Dynamics - Bone: dyn_collar_tag(ComponentPose=@n2, Alpha=1, ExternalForce="0,0,0") {
    angularDampingOverride=0.9
    angularSpringConstant=2000
    bAngularSpring=true
    bOverrideAngularDamping=true
    bUseGravityOverride=true
    gravityScale=0.35
    simulationSpace=RootRelative
    alphaBoolBlend {blendOption=Linear}
    boundBone {boneName=dyn_collar_tag}
    componentAppliedLinearAccClamp=<3000 3000 3000>
    componentLinearAccScale=<2 2 2>
    componentLinearVelScale=<2 2 2>
    gravityOverride {z=-980}
    physicsBodyDefinitions [{
      boundBone {boneName=dyn_collar_tag}
      boxExtents=<40 10 10>
      constraintSetup {angularLimitsMax=<5 30 10>, angularLimitsMin=<-5 -30 -45>, angularTarget=<1 0 0>}
      localJointOffset=<6 0 0>
     }]
   }
   Output Pose(Result=Component To Local(ComponentPose=@n3))
  ]}
 EventGraph {nodes [
   Event BlueprintUpdateAnimation()
   > n1 = Sequence!(execute) -> (then_0, then_1, then_2:Exec, then_3:Exec)
   Set PartSkelMesh(execute=@n1.then_1, PartSkelMesh:Skeletal Mesh Component Object Reference=GetPartSkeletalMeshComponent()) -> (Output_Get:Skeletal Mesh Component Object Reference)
   Branch(execute=@n1.then_2, Condition=Get bClothEnabled())
   > n2 = AlignVectorWithJoint!(
    execute
    Vector:Vector="-1000,400,0"
    MeshComp:Mesh Component Object Reference=Get PartSkelMesh() -> (PartSkelMesh:Skeletal Mesh Component Object Reference)
    PartAnimInstance:Custom Character Part Anim Instance Object Reference
    JointName:Name=hips
   ) -> (then, ReturnValue:Vector)
   > n3 = Get!(execute:Exec) -> (then:Exec, else:Exec, As Chaos Clothing Interactor:Chaos Clothing Interactor Object Reference)
   > SetGravity!(
    execute
    self:Chaos Clothing Interactor Object Reference=@n3
    GravityScale:float=1
    bIsGravityOverridden:bool=true
    GravityOverride:Vector=@n2
   ) -> (then)
   Set RbanOrCloth(execute=@n1.then_3, RbanOrCloth:bool=OR Boolean(A=Get bRigidBodyEnabled(), B=Get bClothEnabled())) -> (Output_Get:bool)
   n4 = GetClothingSimulationInteractor!(self:Skeletal Mesh Component Object Reference=GetOwningComponent(), ClothingAssetName:Name=None) -> (ReturnValue:Clothing Simulation Interactor Object Reference)
   n5 = Is Valid(exec=@n1.then_0, InputObject=@n4)
   n6 = MakeLiteralString(Value=ShirtTilapia_Quadruped_Clothing_0)
   > n7 = Cast To ChaosClothingInteractor!(
    execute
    Object=GetClothingInteractor!(self:Clothing Simulation Interactor Object Reference=@n4, ClothingAssetName:Name=String To Name(InString=@n6)) -> (ReturnValue:Clothing Interactor Object Reference)
   ) -> (then, CastFailed, AsChaos Clothing Interactor:Chaos Clothing Interactor Object Reference) {is=title K2Node_DynamicCast}
   > Set As Chaos Clothing Interactor(As Chaos Clothing Interactor:Chaos Clothing Interactor Object Reference=@n7) -> (Output_Get:Chaos Clothing Interactor Object Reference)
   Set As Chaos Clothing Interactor(execute=[@n7.CastFailed, @n5.Is Not Valid])
   DrawVectorFromJointFunction!(
    execute
    SkelMeshComponent:Scene Component Object Reference=Get PartSkelMesh()
    JointName:Name=hips
    LineEnd:Vector=@n2
    JointOffset:Vector="0,0,50"
    LineColor:Linear Color Structure=(R=1,G=0,B=0,A=1)
    Thickness:double=1
    Foreground:EDrawDebugSceneDepthPriorityGroup Enum=Foreground
   ) -> (then)
   DrawVectorFromJointFunction(
    SkelMeshComponent=Get PartSkelMesh()
    JointName=pelvis
    LineEnd="0, 0, 0"
    JointOffset="0,0,50"
    LineColor=(R=0,G=0,B=1,A=1)
    Thickness=1
    Foreground=Foreground
   )
  ]}
]
variables [
 As Chaos Clothing Interactor {type=ChaosClothingInteractor, default=null}
 PartSkelMesh {type=SkeletalMeshComponent}
 RbanOrCloth {type=boolean}
]
