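# Physics asset for the WhisperRemedy quadruped mesh, defining bodies and constraints for a dynamic three-strand cape.
# Left, middle and right chains each link four capsules from a master bone, with narrow limits; the middle swings mainly on one axis.
# Free-rotation constraints tie each side segment to its middle counterpart, linking the strands like a sheet.
# Hips, spine, front scapulae and rear thighs carry capsules, some with added boxes, apparently as cape colliders.
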PhysicsAsset /WhisperRemedyAssets/Assets/WhisperRemedy/Meshes/WhisperRemedy_Quadruped_Dyn
bodies {
 dyn_cape_l_01 [{shapeType=Capsule, length=5.8076, radius=4.3854, shapeName=dyn_cape_l_01_capsule, center=<3.4966 0 5.541>, rotation=<rot 90 0 0>}]
 dyn_cape_l_01_MSTR [{shapeType=Capsule, length=0.505, radius=0.505, shapeName=dyn_cape_l_01_MSTR_capsule, rotation=<rot 0 -90 0>}]
 dyn_cape_l_02 [{shapeType=Capsule, length=5.8076, radius=4.3854, shapeName=dyn_cape_l_01_capsule, center=<3.4966 0 5.6055>, rotation=<rot 90 0 0>}]
 dyn_cape_l_03 [{shapeType=Capsule, length=5.8076, radius=4.3854, shapeName=dyn_cape_l_01_capsule, center=<3.4966 0 5.2035>, rotation=<rot 90 0 0>}]
 dyn_cape_l_04 [{shapeType=Capsule, length=17.0071, radius=4.3854, shapeName=dyn_cape_l_01_capsule, center=<10.5237 0 5.2035>, rotation=<rot 90 0 0>}]
 dyn_cape_mid_01 [{shapeType=Capsule, length=5.8076, radius=4.3854, shapeName=dyn_cape_l_01_capsule, center=<3.4966 0 4.087>, rotation=<rot 90 0 0>}]
 dyn_cape_mid_01_MSTR [{shapeType=Capsule, length=0.505, radius=0.505, shapeName=dyn_cape_mid_01_MSTR_capsule, rotation=<rot 0 -90 0>}]
 dyn_cape_mid_02 [{shapeType=Capsule, length=5.8076, radius=4.3854, shapeName=dyn_cape_l_01_capsule, center=<3.4966 0 4.6363>, rotation=<rot 90 0 0>}]
 dyn_cape_mid_03 [{shapeType=Capsule, length=5.8076, radius=4.3854, shapeName=dyn_cape_l_01_capsule, center=<3.4966 0 4.6802>, rotation=<rot 90 0 0>}]
 dyn_cape_mid_04 [{shapeType=Capsule, length=17.0071, radius=4.3854, shapeName=dyn_cape_l_01_capsule, center=<10.5237 0 5.3818>, rotation=<rot 90 0 0>}]
 dyn_cape_r_01 [{shapeType=Capsule, length=5.8076, radius=4.3854, shapeName=dyn_cape_l_01_capsule, center=<3.4966 0 5.5282>, rotation=<rot 90 0 0>}]
 dyn_cape_r_01_MSTR [{shapeType=Capsule, length=0.505, radius=0.505, shapeName=dyn_cape_r_01_MSTR_capsule, rotation=<rot 0 -90 0>}]
 dyn_cape_r_02 [{shapeType=Capsule, length=5.8076, radius=4.3854, shapeName=dyn_cape_l_01_capsule, center=<3.4966 0 5.5309>, rotation=<rot 90 0 0>}]
 dyn_cape_r_03 [{shapeType=Capsule, length=5.8076, radius=4.3854, shapeName=dyn_cape_l_01_capsule, center=<3.4966 0 5.1351>, rotation=<rot 90 0 0>}]
 dyn_cape_r_04 [{shapeType=Capsule, length=17.0071, radius=4.3854, shapeName=dyn_cape_l_01_capsule, center=<10.5237 0 5.0676>, rotation=<rot 90 0 0>}]
 hips [
  {shapeType=Capsule, length=0, radius=11.6719, center=<-5.312 1.7717 0>, rotation=<rot 90 176.4711 0>}
  {shapeType=Box, extentX=13.359, extentY=7.5, extentZ=13.0331, shapeName=hips_box, center=<-5.7099 -4.6812 0>, rotation=<rot 90 180 3.5289>}
 ]
 scapula_fr_l [{shapeType=Capsule, length=8.2209, radius=6.4893, center=<3.8593 0 2.4139>, rotation=<rot 71.442 0 0>}]
 scapula_fr_r [{shapeType=Capsule, length=8.2209, radius=6.4893, center=<-3.8593 0 -2.4139>, rotation=<rot 71.442 0 -180>}]
 spine_01 [
  {shapeType=Capsule, length=1.1822, radius=10.7634, center=<6.1736 0.38 0>, rotation=<rot 90 -9.256 0>}
  {shapeType=Box, extentX=13.359, extentY=7.5, extentZ=9.732, shapeName=spine_02_box, center=<5.2042 -5.5682 0>, rotation=<rot 90 -9.256 0>}
 ]
 spine_02 [
  {shapeType=Capsule, length=1.6883, radius=10.7634, center=<6.9686 -1.1423 0>, rotation=<rot 90 -9.256 0>}
  {shapeType=Box, extentX=13.359, extentY=7.5, extentZ=12.8233, shapeName=spine_02_box, center=<5.9993 -7.0905 0>, rotation=<rot 90 -9.256 0>}
 ]
 spine_03 [{shapeType=Capsule, length=0.2758, radius=10.9953, center=<0.6697 -2.4483 0>, rotation=<rot 90 -8.7949 0>}]
 thigh_bk_l [{shapeType=Capsule, length=25.0036, radius=9.2396, center=<11.5944 -1.2436 -3.3004>, rotation=<rot 85.0556 175.2161 180>}]
 thigh_bk_r [{shapeType=Capsule, length=24.4523, radius=9.7909, center=<-11.5678 1.2414 2.9925>, rotation=<rot 82.017 175.2163 0>}]
}
constraints [
 {bone1Name=dyn_cape_l_01, bone2Name=dyn_cape_l_01_MSTR, swing1LimitDegrees=15, swing2LimitDegrees=15, twistLimitDegrees=5}
 {bone1Name=dyn_cape_l_02, bone2Name=dyn_cape_l_01, swing1LimitDegrees=15, swing2LimitDegrees=15, twistLimitDegrees=5}
 {bone1Name=dyn_cape_l_03, bone2Name=dyn_cape_l_02, swing1LimitDegrees=15, swing2LimitDegrees=15, twistLimitDegrees=5}
 {bone1Name=dyn_cape_l_04, bone2Name=dyn_cape_l_03, swing1LimitDegrees=15, swing2LimitDegrees=15, twistLimitDegrees=5}
 {bone1Name=dyn_cape_mid_01, bone2Name=dyn_cape_mid_01_MSTR, swing1LimitDegrees=1, swing2LimitDegrees=45, twistLimitDegrees=5}
 {bone1Name=dyn_cape_mid_02, bone2Name=dyn_cape_mid_01, swing1LimitDegrees=1, swing2LimitDegrees=45, twistLimitDegrees=5}
 {bone1Name=dyn_cape_mid_03, bone2Name=dyn_cape_mid_02, swing1LimitDegrees=1, swing2LimitDegrees=45, twistLimitDegrees=5}
 {bone1Name=dyn_cape_mid_04, bone2Name=dyn_cape_mid_03, swing1LimitDegrees=1, swing2LimitDegrees=45, twistLimitDegrees=5}
 {bone1Name=dyn_cape_r_01, bone2Name=dyn_cape_r_01_MSTR, swing1LimitDegrees=15, swing2LimitDegrees=15, twistLimitDegrees=5}
 {bone1Name=dyn_cape_r_02, bone2Name=dyn_cape_r_01, swing1LimitDegrees=15, swing2LimitDegrees=15, twistLimitDegrees=5}
 {bone1Name=dyn_cape_r_03, bone2Name=dyn_cape_r_02, swing1LimitDegrees=15, swing2LimitDegrees=15, twistLimitDegrees=5}
 {bone1Name=dyn_cape_r_04, bone2Name=dyn_cape_r_03, swing1LimitDegrees=15, swing2LimitDegrees=15, twistLimitDegrees=5}
 {
  bone1Name=dyn_cape_l_01
  bone2Name=dyn_cape_mid_01
  swing1LimitDegrees=45
  swing1Motion=Free
  swing2LimitDegrees=45
  swing2Motion=Free
  twistLimitDegrees=45
  twistMotion=Free
 }
 {
  bone1Name=dyn_cape_r_01
  bone2Name=dyn_cape_mid_01
  swing1LimitDegrees=45
  swing1Motion=Free
  swing2LimitDegrees=45
  swing2Motion=Free
  twistLimitDegrees=45
  twistMotion=Free
 }
 {
  bone1Name=dyn_cape_r_02
  bone2Name=dyn_cape_mid_02
  swing1LimitDegrees=45
  swing1Motion=Free
  swing2LimitDegrees=45
  swing2Motion=Free
  twistLimitDegrees=45
  twistMotion=Free
 }
 {
  bone1Name=dyn_cape_l_02
  bone2Name=dyn_cape_mid_02
  swing1LimitDegrees=45
  swing1Motion=Free
  swing2LimitDegrees=45
  swing2Motion=Free
  twistLimitDegrees=45
  twistMotion=Free
 }
 {
  bone1Name=dyn_cape_r_03
  bone2Name=dyn_cape_mid_03
  swing1LimitDegrees=45
  swing1Motion=Free
  swing2LimitDegrees=45
  swing2Motion=Free
  twistLimitDegrees=45
  twistMotion=Free
 }
 {
  bone1Name=dyn_cape_l_03
  bone2Name=dyn_cape_mid_03
  swing1LimitDegrees=45
  swing1Motion=Free
  swing2LimitDegrees=45
  swing2Motion=Free
  twistLimitDegrees=45
  twistMotion=Free
 }
 {
  bone1Name=dyn_cape_r_04
  bone2Name=dyn_cape_mid_04
  swing1LimitDegrees=45
  swing1Motion=Free
  swing2LimitDegrees=45
  swing2Motion=Free
  twistLimitDegrees=45
  twistMotion=Free
 }
 {
  bone1Name=dyn_cape_l_04
  bone2Name=dyn_cape_mid_04
  swing1LimitDegrees=45
  swing1Motion=Free
  swing2LimitDegrees=45
  swing2Motion=Free
  twistLimitDegrees=45
  twistMotion=Free
 }
 {
  bone1Name=spine_02
  bone2Name=spine_01
  swing1LimitDegrees=45
  swing1Motion=Free
  swing2LimitDegrees=45
  swing2Motion=Free
  twistLimitDegrees=45
  twistMotion=Free
 }
]
